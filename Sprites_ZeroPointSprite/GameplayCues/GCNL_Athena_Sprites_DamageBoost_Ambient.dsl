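; Looping gameplay cue for the Sprites damage boost that shows an ambient effect on the player's weapon muzzle.
; On loop start it stores the first particle component and the player pawn's current weapon.
; For a ranged weapon it shows the effect and attaches it to the Muzzle socket; other valid weapons hide it.
; A weapon-equipped delegate repeats this check so the effect follows weapon swaps.
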(event EventBeginPlay)

(event Collision|EventActorBeginOverlap
  (params (OtherActor :type Actor%20Object%20Reference)))

(event EventTick
  (params (DeltaSeconds :type Float%20%28single-precision%29)))

(event EventOnLoopingStart(EitherVFX-preferthisversion)
  (params (MyTarget :type Actor%20Object%20Reference) (Parameters :type Gameplay%20Cue%20Parameters%20Structure%20%28by%20ref%29) (ParticleComponents :type Array%20of%20FXSystem%20Component%20Object%20References) (AudioComponents :type Array%20of%20Audio%20Component%20Object%20References))
  (bind _weaponidlefx (Variables|Default|GetWeaponIdleFX))
  (|CallParentFunction MyTarget Parameters ParticleComponents AudioComponents)
  (Variables|Default|SetWeaponIdleFX (Utilities|Array|Get(acopy) ParticleComponents))
  (bind _asfort_player_pawn (Utilities|Casting|CastToFortPlayerPawn MyTarget))
  (Utilities|FlowControl|Sequence
    (:then_0
      (bind _output_get (Variables|Default|SetCurrentWeapon (Weapon|GetCurrentWeapon _asfort_player_pawn)))
      (Utilities|IsValid _output_get
        (:Is%20Valid
          (Utilities|Casting|CastToFortWeaponRanged _output_get
            (:then
              (Rendering|SetVisibility _weaponidlefx true false)
              (Transformation|AttachComponentToComponent _weaponidlefx (Variables|/Script/FortniteGame.FortWeapon|GetWeaponMesh _output_get) "Muzzle" "KeepRelative" "KeepRelative" "KeepRelative" true))
            (:CastFailed
              (Rendering|SetVisibility _weaponidlefx false false))))))
    (:then_1
      (FortPawn|AssignOnWeaponEquippedDelegate _asfort_player_pawn (delegate Custom|OnWeaponEquippedDelegate_Event)))))

(event Custom|OnWeaponEquippedDelegate_Event
  (params (NewWeapon :type Fort%20Weapon%20Object%20Reference) (PrevWeapon :type Fort%20Weapon%20Object%20Reference))
  (bind _weaponidlefx (Variables|Default|GetWeaponIdleFX))
  (bind _currentweapon (Variables|Default|GetCurrentWeapon))
  (bind _output_get (Variables|Default|SetCurrentWeapon NewWeapon))
  (Utilities|IsValid _output_get
    (:Is%20Valid
      (Utilities|Casting|CastToFortWeaponRanged _currentweapon
        (:then
          (Rendering|SetVisibility _weaponidlefx true false)
          (Transformation|AttachComponentToComponent _weaponidlefx (Variables|/Script/FortniteGame.FortWeapon|GetWeaponMesh _currentweapon) "Muzzle" "KeepRelative" "KeepRelative" "KeepRelative" true))
        (:CastFailed
          (Rendering|SetVisibility _weaponidlefx false false))))))

(fn ConstructionScript)
