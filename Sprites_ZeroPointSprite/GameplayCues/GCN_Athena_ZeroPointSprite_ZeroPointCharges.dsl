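; Looping gameplay cue that displays Zero Point Sprite dash charges as a ring of Niagara effects attached to the player.
; It spawns one effect per configured charge, then activates them to match the simulated charge attribute, adding bursts and one gated sound for regained charges.
; Gameplay tag tasks independently toggle visibility for vehicle, aiming and IcicleLarva states, and set a Niagara parameter during tactical sprint.
; Removal destroys the effects and cancels the attribute listener.
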(event EventOnLoopingStart(EitherVFX-preferthisversion)
  (params (MyTarget :type Actor%20Object%20Reference) (Parameters :type Gameplay%20Cue%20Parameters%20Structure%20%28by%20ref%29) (ParticleComponents :type Array%20of%20FXSystem%20Component%20Object%20References) (AudioComponents :type Array%20of%20Audio%20Component%20Object%20References))
  (bind _row_numcharges (Variables|Default|GetRow_NumCharges))
  (bind _returnvalue (ScalableFloat|GetValueatLevel _row_numcharges))
  (bind _returnvalue_2 (Math|Float|Truncate _returnvalue))
  (|CallParentFunction MyTarget Parameters ParticleComponents AudioComponents)
  (bind _asfort_player_pawn (Utilities|Casting|CastToFortPlayerPawn MyTarget))
  (Variables|Default|SetPlayerPawn _asfort_player_pawn)
  (for _index (range _returnvalue_2)
    ;[0.9271 0.6954 0.2539 0.5] FX version of the Charges
    (bind _returnvalue_3 (Niagara|SpawnSystemAttached "/Sprites_ZeroPointSprite/Assets/FX/NS_SuperSprite_DashCharge_Idle.NS_SuperSprite_DashCharge_Idle" (Variables|Transformation|GetRootComponent) "None" (Utilities|Operators|Add (Math|Vector|RotateVectorAroundAxis "35.000000,0.000000,0.000000" (+ (* (/ (Math|Conversions|ToFloat(Integer) _index) _returnvalue_2) 3.6e+02) 9e+01) "0.000000,0.000000,1.000000") (0.0 0.0 75.0)) "0, 0, 0" "KeepRelativeOffset" false false))
    (Utilities|Array|Add (Variables|Default|GetChargesFX) _returnvalue_3)
    (CallFunction|UpdateCharges))
  (Utilities|FlowControl|Sequence
    (:then_0
      (bind _asynctaskproxy (Ability|Tasks|WaitForSimulatedAttributeChanged MyTarget (Variables|Default|GetCurrentChargesAttribute))
        (:then
          (Variables|Default|SetAsync%20Action _asynctaskproxy))
        (:Changed
          (CallFunction|UpdateCharges))))
    (:then_1
      ;[0.2774 1 0.8999 0.5] Tweak FX under specific circumstances
      (Utilities|FlowControl|Sequence
        (:then_0
          (CallFunction|VehicleStateControl))
        (:then_1
          (CallFunction|SprintStateControl))
        (:then_2
          (CallFunction|ADSStateControl))
        (:then_3
          (CallFunction|IcicleLarvaStateControl))))))

(event EventOnRemoval(EitherVFX-preferthisversion)
  (params (MyTarget :type Actor%20Object%20Reference) (Parameters :type Gameplay%20Cue%20Parameters%20Structure%20%28by%20ref%29) (ParticleComponents :type Array%20of%20FXSystem%20Component%20Object%20References) (AudioComponents :type Array%20of%20Audio%20Component%20Object%20References) (BurstCameraShakeInstance :type Camera%20Shake%20Base%20Object%20Reference) (BurstDecalInstance :type Decal%20Actor%20Object%20Reference))
  (bind _chargesfx (Variables|Default|GetChargesFX))
  ;[0.9271 0.6954 0.2539 0.5] Same Implementation but using the FX version for teh Charges
  (for _array_element _chargesfx
    (Components|DestroyComponent _array_element 0))
  (Utilities|Array|Clear _chargesfx)
  (bind _async_action (Variables|Default|GetAsync%20Action)
    (:then
      (AsyncAction|Cancel _async_action))
    (:else)))

(event Custom|UpdateCharges
  (bind _currentchargesattribute (Variables|Default|GetCurrentChargesAttribute))
  (bind _playerpawn (Variables|Default|GetPlayerPawn))
  (bind _returnvalue (GadgetAttributeSets|GetSimulatedAttributeValue _playerpawn _currentchargesattribute 0.0 :!class /Script/FortniteGame.FortPlayerPawn))
  (bind _returnvalue_2 (Math|Float|Truncate _returnvalue))
  ;[0.9271 0.6954 0.2539 0.5] Same Implementation but using the FX version for teh Charges
  (for (_array_element _array_index) (Variables|Default|GetChargesFX)
    (Utilities|FlowControl|Sequence
      (:then_0
        (if (and (not (Components|Activation|IsActive _array_element)) (< _array_index _returnvalue_2))
          (Niagara|SpawnSystemAttached "/Sprites_ZeroPointSprite/Assets/FX/NS_SuperSprite_DashCharge_SpawnBurst.NS_SuperSprite_DashCharge_SpawnBurst" _array_element "None" "0, 0, 0" "0, 0, 0" "KeepRelativeOffset" true)
          (Utilities|FlowControl|Gate true
            (:Exit
              (if (not (Variables|Default|GetisIcicleLarva))
                (SoundLibrary|PlaySoundSimple (WorldSubsystems|GetSoundLibrarySubsystem (Variables|Default|GetPlayerPawn)) (Variables|Default|GetPlayerPawn) (Utilities|Struct|MakeSimpleSoundSettings (Variables|Default|GetEvent%20Name)) :!class /Script/SoundLibrary.SoundLibrarySubsystem)))) ; Only play a sound once if any charges spawned
        ))
      (:then_1
        (Components|Activation|SetActive _array_element (< _array_index _returnvalue_2) false))))
  (Utilities|FlowControl|Sequence
    (:then_0
      _)
    (:then_1
      _)))

(event Custom|VehicleStateControl
  (bind _playerpawn (Variables|Default|GetPlayerPawn))
  ;[1 0.5608 0.2962 0.5] Check if Player Enters/Exits a Vehicle
  (Development|PrintString 0 "VEHICLE STATE CONTROL" true true "(R=1.000000,G=0.314476,B=0.901354,A=1.000000)")
  (Ability|Tasks|WaitGameplayTagAddToActor _playerpawn "(TagName=\"Gameplay.Status.InVehicle\")"
    (:then
      (Ability|Tasks|WaitGameplayTagRemoveFromActor _playerpawn "(TagName=\"Gameplay.Status.InVehicle\")"
        (:Removed
          (Development|PrintString 0 "NOT IN A VEHICLE" true true "(R=1.000000,G=0.059447,B=0.068042,A=1.000000)")
          (bind _output_get (Variables|Default|SetisInVehicle false))
          (CallFunction|HideChargeFX _output_get))))
    (:Added
      (Development|PrintString 0 "IN A VEHICLE" true true "(R=0.145127,G=1.000000,B=0.232407,A=1.000000)")
      (bind _output_get_2 (Variables|Default|SetisInVehicle true))
      (CallFunction|HideChargeFX _output_get_2))))

(event Custom|IcicleLarvaStateControl
  (bind _playerpawn (Variables|Default|GetPlayerPawn))
  (Ability|Tasks|WaitGameplayTagAddToActor _playerpawn "(TagName=\"Granted.IcicleLarva.Equipped.Special\")"
    (:then
      (Ability|Tasks|WaitGameplayTagRemoveFromActor _playerpawn "(TagName=\"Granted.IcicleLarva.Equipped.Special\")"
        (:then
          (if (GameplayTags|HasMatchingGameplayTag _playerpawn "(TagName=\"Granted.IcicleLarva.Equipped.Special\")")
            (bind _output_get (Variables|Default|SetisIcicleLarva true))
            (CallFunction|HideChargeFX _output_get)))
        (:Removed
          (bind _output_get_2 (Variables|Default|SetisIcicleLarva false))
          (CallFunction|HideChargeFX _output_get_2))))
    (:Added
      (bind _output_get_3 (Variables|Default|SetisIcicleLarva true))
      (CallFunction|HideChargeFX _output_get_3))))

(event Custom|HideChargeFX
  (params (HideFX :type Boolean))
  ;[0.9583 0.3144 0.1671 0.5] Hide Charges FX
  (for _array_element (Variables|Default|GetChargesFX)
    (Rendering|SetVisibility _array_element (not HideFX) false)))

(event Custom|ADSStateControl
  (bind _playerpawn (Variables|Default|GetPlayerPawn))
  ;[0.6823 0.4332 0.2002 0.5] Check if Player is aiming gun (ADS)
  (Ability|Tasks|WaitGameplayTagAddToActor _playerpawn "(TagName=\"Gameplay.Status.IsTargeting\")"
    (:then
      (Ability|Tasks|WaitGameplayTagRemoveFromActor _playerpawn "(TagName=\"Gameplay.Status.IsTargeting\")"
        (:Removed
          (bind _output_get (Variables|Default|SetisADS false))
          (CallFunction|HideChargeFX _output_get))))
    (:Added
      (bind _output_get_2 (Variables|Default|SetisADS true))
      (CallFunction|HideChargeFX _output_get_2))))

(event Custom|SprintStateControl
  (bind _playerpawn (Variables|Default|GetPlayerPawn))
  ;[0.2515 0.9531 0.4256 0.5] Check if Player is Tactical Sprinting
  (Ability|Tasks|WaitGameplayTagAddToActor _playerpawn "(TagName=\"Gameplay.Action.Player.TacticalSprint\")"
    (:then
      (Ability|Tasks|WaitGameplayTagRemoveFromActor _playerpawn "(TagName=\"Gameplay.Action.Player.TacticalSprint\")"
        (:Removed
          (Development|PrintString 0 "NOT SPRINTING" true true "(R=1.000000,G=0.248831,B=0.236894,A=1.000000)")
          (CallFunction|PlayerTacticalSprintFXTweaks false))))
    (:Added
      (Development|PrintString 0 "SPRINTING" true true "(R=1.000000,G=0.986926,B=0.369560,A=1.000000)")
      (CallFunction|PlayerTacticalSprintFXTweaks true))))

(event Custom|PlayerTacticalSprintFXTweaks
  (params (isPlayerTacticalSprinting :type Boolean))
  ;[0.2222 0.8802 0.6597 0.5] Tweak FX Camera Offset when Sprinting
  (for _array_element (Variables|Default|GetChargesFX)
    (if isPlayerTacticalSprinting
      (Niagara|SetNiagaraVariable(Float) _array_element "PlayerTacticalSprinting" 1.0)
      (else
        (Niagara|SetNiagaraVariable(Float) _array_element "PlayerTacticalSprinting" 0.0)))))

(fn ConstructionScript)
