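; Burst gameplay cue that plays the bounce sound for the Zero Point Sprite projectile.
; On burst it casts the target to Prj_Athena_ZeroPointSprite and spawns the sprite bounce sound at that actor's location.
; It passes the projectile's bounce count minus one to the sound as the BounceCount integer parameter, then fires the Bounce trigger, letting the audio asset respond to each successive bounce.
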(event EventOnBurst
  (params (Target :type Actor%20Object%20Reference) (Parameters :type Gameplay%20Cue%20Parameters%20Structure%20%28by%20ref%29) (SpawnResults :type Gameplay%20Cue%20Notify%20Spawn%20Result%20Structure%20%28by%20ref%29))
  (bind _asprj_athena_zero_point_sprite (Utilities|Casting|CastToPrj_Athena_ZeroPointSprite Target))
  (bind _returnvalue (Audio|SpawnSoundatLocation _asprj_athena_zero_point_sprite "/Sprites_ZeroPointSprite/Assets/Sounds/Cues/SFX/MSS_Sidekick_Sprite_Bounce.MSS_Sidekick_Sprite_Bounce" (Transformation|GetActorLocation Target)))
  (Audio|Parameter|SetIntegerParameter _returnvalue "BounceCount" (- (Variables|/Sprites/Gameplay/ItemVersion/Abilities/Prj_Athena_Sprite_Parent.Prj_Athena_Sprite_Parent_C|GetBounce%20Count _asprj_athena_zero_point_sprite) 1) :!class /Script/AudioExtensions.AudioParameterControllerInterface)
  (Audio|Parameter|ExecuteTriggerParameter _returnvalue "Bounce"))

(fn ConstructionScript)
